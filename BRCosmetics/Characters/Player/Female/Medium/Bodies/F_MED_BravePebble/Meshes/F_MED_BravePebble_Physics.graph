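# Physics asset for the F_MED_BravePebble female cosmetic body, covering a dynamic holster bone chain plus the pelvis, lower spine and thighs.
# The holster runs from a master bone on the pelvis through three links with tight twist and swing limits that loosen in one swing axis toward the tip.
# All shapes are capsules, with two long ones on the last holster link.
# Torso, thigh and holster master constraints leave angular motion free.
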PhysicsAsset /BRCosmetics/Characters/Player/Female/Medium/Bodies/F_MED_BravePebble/Meshes/F_MED_BravePebble_Physics
bodies {
 dyn_holster_01 [{shapeType=Capsule, length=2, radius=2, shapeName=dyn_holster_01_capsule, center=<2 0 0>}]
 dyn_holster_02 [{shapeType=Capsule, length=2, radius=2, shapeName=dyn_holster_02_capsule, center=<2 0 0>}]
 dyn_holster_03 [
  {
   shapeType=Capsule
   length=26.223
   radius=2.1599
   shapeName=dyn_holster_03_capsule
   center=<3.3673 -2.3066 -1.2138>
   rotation=<rot 11.3599 93.6624 -78.1563>
  }
  {
   shapeType=Capsule
   length=26.3801
   radius=2
   shapeName=dyn_holster_03_capsule
   center=<3.2691 -2.7532 2.2875>
   rotation=<rot -0.0115 93.6107 -93.7194>
  }
 ]
 dyn_holster_MSTR [{shapeType=Capsule, length=0.505, radius=0.505, shapeName=dyn_holster_MSTR_capsule, rotation=<rot 0 -90 0>}]
 pelvis [{shapeType=Capsule, length=2.0685, radius=14.2598, rotation=<rot 90 0 0>}]
 spine_01 [{shapeType=Capsule, length=3.727, radius=10.626, shapeName=spine_01_capsule, rotation=<rot 90 0 0>}]
 spine_02 [{shapeType=Capsule, length=4.7358, radius=9.0027, shapeName=spine_02_capsule, rotation=<rot 90 0 0>}]
 thigh_l [{shapeType=Capsule, length=20.8405, radius=9.439, center=<-22.8767 -1.1657 -0.5528>, rotation=<rot 0 88.4238 -86.2946>}]
 thigh_r [{shapeType=Capsule, length=20.8465, radius=9.4024, center=<22.7874 1.1918 0.5451>, rotation=<rot 0 88.869 -87.016>}]
}
constraints [
 {
  bone1Name=thigh_l
  bone2Name=pelvis
  swing1LimitDegrees=45
  swing1Motion=Free
  swing2LimitDegrees=45
  swing2Motion=Free
  twistLimitDegrees=45
  twistMotion=Free
 }
 {
  bone1Name=thigh_r
  bone2Name=pelvis
  swing1LimitDegrees=45
  swing1Motion=Free
  swing2LimitDegrees=45
  swing2Motion=Free
  twistLimitDegrees=45
  twistMotion=Free
 }
 {
  bone1Name=spine_01
  bone2Name=pelvis
  swing1LimitDegrees=45
  swing1Motion=Free
  swing2LimitDegrees=45
  swing2Motion=Free
  twistLimitDegrees=45
  twistMotion=Free
 }
 {bone1Name=dyn_holster_03, bone2Name=dyn_holster_02, swing1LimitDegrees=13, swing2LimitDegrees=22, twistLimitDegrees=3}
 {
  bone1Name=dyn_holster_MSTR
  bone2Name=pelvis
  swing1LimitDegrees=45
  swing1Motion=Free
  swing2LimitDegrees=45
  swing2Motion=Free
  twistLimitDegrees=45
  twistMotion=Free
 }
 {bone1Name=dyn_holster_01, bone2Name=dyn_holster_MSTR, swing1LimitDegrees=13, swing2LimitDegrees=10, twistLimitDegrees=3}
 {bone1Name=dyn_holster_02, bone2Name=dyn_holster_01, swing1LimitDegrees=13, swing2LimitDegrees=18, twistLimitDegrees=3}
 {
  bone1Name=spine_02
  bone2Name=spine_01
  swing1LimitDegrees=45
  swing1Motion=Free
  swing2LimitDegrees=45
  swing2Motion=Free
  twistLimitDegrees=45
  twistMotion=Free
 }
]
